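; Conversation node GrantPlayerDuel in the Tournament bounty's DuelistPlayer conversations, with only its editor body color defined here.
; GetNodeBodyColor returns the debug participant color for an empty participant tag and reports success.
; The file shows no logic for granting the duel itself.
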(fn GetNodeBodyColor
  (return (Conversation|GetDebugParticipantColor "(TagName=\"\")") true))
